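# Sound cue for the first-person firing sound of the grappling hook weapon.
# A Random node picks among four equally weighted Grappling_Fire wave variations, and is excluded from branch culling.
# Playback uses a first-person weapon projectile attenuation asset, a ranged-weapon 1P fire sound class and a shared first-person gunfire concurrency set, with a boosted volume.
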SoundCue /Game/Sounds/Fort_Weapons/GrapplingHook/Grappling_Fire_1P_Cue
root=Random_1
nodes [
 {
  id=Random_1
  children [WavePlayer_2,WavePlayer_3,WavePlayer_4,WavePlayer_5]
  properties {bShouldExcludeFromBranchCulling=true, weights [1,1,1,1]}
 }
 {id=WavePlayer_2, properties {soundWaveAssetPtr=&/Grappling_Fire_01}}
 {id=WavePlayer_3, properties {soundWaveAssetPtr=&/Grappling_Fire_02}}
 {id=WavePlayer_4, properties {soundWaveAssetPtr=&/Grappling_Fire_03}}
 {id=WavePlayer_5, properties {soundWaveAssetPtr=&/Grappling_Fire_04}}
]
properties {
 attenuationSettings=&Fort_Weapon_Projectile_FirstPerson_Athena_Att
 maxDistance=2097152
 soundClassObject=&Fort_WeaponRanged_Fire_1P_PSM
 volumeMultiplier=1.35
 attenuationOverrides {falloffDistance=5000, attenuationShapeExtents {x=1000}}
 concurrencyOverrides {resolutionRule=PreventNew}
 concurrencySet [&Weapon_Gunfire_FirstPerson_Conc]
}
