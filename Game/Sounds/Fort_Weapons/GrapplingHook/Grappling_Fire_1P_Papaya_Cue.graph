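# First-person grappling hook fire sound cue (Papaya variant), classed as a ranged weapon fire sound with first-person gunfire concurrency.
# A quality-level switch scales variety: the first level randomizes four fire waves, the second randomizes two, and the last two share one fixed wave.
# Each branch passes through a mixer that lowers its level and has an empty second input.
# The cue also slightly lowers pitch, reduces volume and uses first-person projectile attenuation.
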SoundCue /Game/Sounds/Fort_Weapons/GrapplingHook/Grappling_Fire_1P_Papaya_Cue
root=QualityLevel_0
nodes [
 {id=Mixer_0, children [Random_4,null], properties {inputVolume [0.8,0.8]}}
 {id=Mixer_1, children [Random_1,null], properties {inputVolume [0.8,0.8]}}
 {id=Mixer_3, children [WavePlayer_17, null], properties {inputVolume [0.8,0.8]}}
 {id=QualityLevel_0, children [Mixer_1,Mixer_0,Mixer_3,Mixer_3]}
 {id=Random_1, children [WavePlayer_2,WavePlayer_3,WavePlayer_4,WavePlayer_5], properties {weights [1,1,1,1]}}
 {id=Random_4, children [WavePlayer_15, WavePlayer_16], properties {weights [1,1]}}
 {id=WavePlayer_15, properties {soundWaveAssetPtr=&/Grappling_Fire_02}}
 {id=WavePlayer_16, properties {soundWaveAssetPtr=&/Grappling_Fire_03}}
 {id=WavePlayer_17, properties {soundWaveAssetPtr=&Grappling_Fire_02}}
 {id=WavePlayer_2, properties {soundWaveAssetPtr=&/Grappling_Fire_01}}
 {id=WavePlayer_3, properties {soundWaveAssetPtr=&Grappling_Fire_02}}
 {id=WavePlayer_4, properties {soundWaveAssetPtr=&Grappling_Fire_03}}
 {id=WavePlayer_5, properties {soundWaveAssetPtr=&/Grappling_Fire_04}}
]
properties {
 attenuationSettings=&Fort_Weapon_Projectile_FirstPerson_Athena_Att
 maxDistance=2097152
 pitchMultiplier=0.95
 soundClassObject=&Fort_WeaponRanged_Fire_1P_PSM
 volumeMultiplier=0.6
 attenuationOverrides {falloffDistance=5000, attenuationShapeExtents {x=1000}}
 concurrencyOverrides {resolutionRule=PreventNew}
 concurrencySet [&Weapon_Gunfire_FirstPerson_Conc]
}
