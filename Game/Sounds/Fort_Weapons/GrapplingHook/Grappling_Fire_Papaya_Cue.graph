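# Sound cue for the Papaya variant of the grappling hook fire sound, with reduced overall volume and slightly lowered pitch.
# Each quality branch mixes a close layer under the 3P fire sound class with a medium-range mono layer under the Distant class, each with its own attenuation.
# The first quality branch randomizes four waves per layer, the second two, and the last two share a single-wave mix.
# Local and global gunfire concurrency sets limit playback.
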SoundCue /Game/Sounds/Fort_Weapons/GrapplingHook/Grappling_Fire_Papaya_Cue
root=QualityLevel_0
nodes [
 {id=Attenuation_0, children [Random_4], properties {attenuationSettings=&Fort_Weapon_Projectile_Close_Athena_Att}}
 {
  id=Attenuation_1
  children [Random_6]
  properties {attenuationSettings=&Fort_Weapon_Projectile_Medium_Mono_Athena_Att}
 }
 {
  id=Attenuation_2
  children [WavePlayer_17]
  properties {attenuationSettings=&Fort_Weapon_Projectile_Close_Athena_Att}
 }
 {id=Attenuation_3, children [Random_1], properties {attenuationSettings=&Fort_Weapon_Projectile_Close_Athena_Att}}
 {
  id=Attenuation_4
  children [Random_2]
  properties {attenuationSettings=&Fort_Weapon_Projectile_Medium_Mono_Athena_Att}
 }
 {
  id=Attenuation_9
  children [WavePlayer_26]
  properties {attenuationSettings=&Fort_Weapon_Projectile_Medium_Mono_Athena_Att}
 }
 {id=Mixer_0, children [SoundClass_1,SoundClass_0], properties {inputVolume [0.8,0.8]}}
 {id=Mixer_1, children [SoundClass_3,SoundClass_2], properties {inputVolume [0.8,0.8]}}
 {id=Mixer_3, children [SoundClass_7,SoundClass_6], properties {inputVolume [0.8,0.8]}}
 {id=QualityLevel_0, children [Mixer_1,Mixer_0,Mixer_3,Mixer_3]}
 {id=Random_1, children [WavePlayer_2,WavePlayer_3,WavePlayer_4,WavePlayer_5], properties {weights [1,1,1,1]}}
 {id=Random_2, children [WavePlayer_6, WavePlayer_8, WavePlayer_9, WavePlayer_14], properties {weights [1,1,1,1]}}
 {id=Random_4, children [WavePlayer_15, WavePlayer_16], properties {weights [1,1]}}
 {id=Random_6, children [WavePlayer_19, WavePlayer_20], properties {weights [1,1]}}
 {
  id=SoundClass_0
  children [Attenuation_1]
  properties {soundClassOverride=&Fort_WeaponRanged_Fire_Distant}
 }
 {
  id=SoundClass_1
  children [Attenuation_0]
  properties {soundClassOverride=&Fort_WeaponRanged_Fire_3P_PSM}
 }
 {
  id=SoundClass_2
  children [Attenuation_4]
  properties {soundClassOverride=&Fort_WeaponRanged_Fire_Distant}
 }
 {
  id=SoundClass_3
  children [Attenuation_3]
  properties {soundClassOverride=&Fort_WeaponRanged_Fire_3P_PSM}
 }
 {
  id=SoundClass_6
  children [Attenuation_9]
  properties {soundClassOverride=&Fort_WeaponRanged_Fire_Distant}
 }
 {
  id=SoundClass_7
  children [Attenuation_2]
  properties {soundClassOverride=&Fort_WeaponRanged_Fire_3P_PSM}
 }
 {id=WavePlayer_14, properties {soundWaveAssetPtr=&/Grappling_Fire_Medium_04}}
 {id=WavePlayer_15, properties {soundWaveAssetPtr=&/Grappling_Fire_02}}
 {id=WavePlayer_16, properties {soundWaveAssetPtr=&/Grappling_Fire_03}}
 {id=WavePlayer_17, properties {soundWaveAssetPtr=&Grappling_Fire_02}}
 {id=WavePlayer_19, properties {soundWaveAssetPtr=&/Grappling_Fire_Medium_02}}
 {id=WavePlayer_2, properties {soundWaveAssetPtr=&/Grappling_Fire_01}}
 {id=WavePlayer_20, properties {soundWaveAssetPtr=&/Grappling_Fire_Medium_03}}
 {id=WavePlayer_26, properties {soundWaveAssetPtr=&Grappling_Fire_Medium_02}}
 {id=WavePlayer_3, properties {soundWaveAssetPtr=&Grappling_Fire_02}}
 {id=WavePlayer_4, properties {soundWaveAssetPtr=&Grappling_Fire_03}}
 {id=WavePlayer_5, properties {soundWaveAssetPtr=&/Grappling_Fire_04}}
 {id=WavePlayer_6, properties {soundWaveAssetPtr=&/Grappling_Fire_Medium_01}}
 {id=WavePlayer_8, properties {soundWaveAssetPtr=&Grappling_Fire_Medium_02}}
 {id=WavePlayer_9, properties {soundWaveAssetPtr=&Grappling_Fire_Medium_03}}
]
properties {
 maxDistance=10579.2871
 pitchMultiplier=0.95
 volumeMultiplier=0.6
 attenuationOverrides {falloffDistance=5000, attenuationShapeExtents {x=1000}}
 concurrencyOverrides {resolutionRule=PreventNew}
 concurrencySet [&Weapon_Gunfire_Local_Conc, &Weapon_Gunfire_Global_Conc]
}
