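# Sound cue for the grappling hook's player impact sound.
# It picks one of three equally weighted waves, slightly varies and boosts the volume, and starts after a short fixed delay.
# Overridden attenuation switches off distance falloff, spatialization and reverb send, so it plays as 2D despite the assigned grappling attenuation asset.
# It routes to the Impacts sound class and shares the rifle bullet impact concurrency group.
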SoundCue /Game/Sounds/Fort_Weapons/GrapplingHook/Grappling_Impact_Player_Cue
root=Delay_0
nodes [
 {id=Delay_0, children [Modulator_0], properties {delayMax=0.05, delayMin=0.05}}
 {id=Modulator_0, children [Random_1], properties {volumeMax=1.2, volumeMin=1.1}}
 {id=Random_1, children [WavePlayer_4,WavePlayer_5,WavePlayer_6], properties {weights [1,1,1]}}
 {id=WavePlayer_4, properties {soundWaveAssetPtr=&/Grappling_Impact_Player_01}}
 {id=WavePlayer_5, properties {soundWaveAssetPtr=&/Grappling_Impact_Player_02}}
 {id=WavePlayer_6, properties {soundWaveAssetPtr=&/Grappling_Impact_Player_03}}
]
properties {
 attenuationSettings=/Game/Sounds/Attenuation/Fort_GadgetGrappling_Att
 bOverrideAttenuation=true
 maxDistance=2097152
 soundClassObject=&Fort_Impacts
 volumeMultiplier=0.9
 attenuationOverrides {
  bAttenuate=false
  bEnableReverbSend=false
  bSpatialize=false
  falloffDistance=3000
  nonSpatializedRadiusStart=1000
  stereoSpread=2000
  attenuationShapeExtents {x=2500}
 }
 concurrencySet [&RifleBulletImpact_Conc]
}
