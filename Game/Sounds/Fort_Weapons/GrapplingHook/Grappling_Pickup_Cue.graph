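# Sound cue for picking up the grappling hook weapon.
# A random node picks between two pickup wave variants with equal weight.
# It uses weapon pickup attenuation and the loot-collect sound class, with a slight volume boost.
# A concurrency override allows one instance at a time and blocks new plays while one is active.
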SoundCue /Game/Sounds/Fort_Weapons/GrapplingHook/Grappling_Pickup_Cue
root=Random_0
nodes [
 {id=Random_0, children [WavePlayer_0,WavePlayer_1], properties {weights [1,1]}}
 {id=WavePlayer_0, properties {soundWaveAssetPtr=&/GrapplingHook_Pickup_01}}
 {id=WavePlayer_1, properties {soundWaveAssetPtr=&/GrapplingHook_Pickup_02}}
]
properties {
 attenuationSettings=&Fort_Weapon_Pickup_Att
 bOverrideConcurrency=true
 maxDistance=3500
 soundClassObject=&Fort_Loot_Collect
 volumeMultiplier=1.25
 concurrencyOverrides {maxCount=1, resolutionRule=PreventNew}
}
