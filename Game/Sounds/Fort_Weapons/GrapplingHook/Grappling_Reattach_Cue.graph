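# Sound cue for the grappling hook weapon's reattach sound.
# A Random node picks one of three equally weighted reattach waves, and a Modulator varies its pitch slightly upward on each play while holding volume constant.
# The cue uses grappling gadget attenuation, the gadgets and items sound class, a reduced volume multiplier, and a concurrency override that prevents new instances once the limit is reached.
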SoundCue /Game/Sounds/Fort_Weapons/GrapplingHook/Grappling_Reattach_Cue
root=Modulator_13
nodes [
 {id=Modulator_13, children [Random_0], properties {pitchMax=1.2, pitchMin=1, volumeMax=1, volumeMin=1}}
 {id=Random_0, children [WavePlayer_0,WavePlayer_1,WavePlayer_2], properties {bShouldExcludeFromBranchCulling=true, weights [1,1,1]}}
 {id=WavePlayer_0, properties {soundWaveAssetPtr=&/Grappling_Reattach_01}}
 {id=WavePlayer_1, properties {soundWaveAssetPtr=&/Grappling_Reattach_02}}
 {id=WavePlayer_2, properties {soundWaveAssetPtr=&/Grappling_Reattach_03}}
]
properties {
 attenuationSettings=/Game/Sounds/Attenuation/Fort_GadgetGrappling_Att
 bOverrideConcurrency=true
 maxDistance=8500
 soundClassObject=&Fort_Gadgets_Items
 volumeMultiplier=0.3
 concurrencyOverrides {resolutionRule=PreventNew}
}
