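# Sound cue for the grappling hook's reload insert foley (Papaya variant), built from reused crossbow reload recordings.
# A Random node picks evenly between two crossbow reload waves, and a Modulator adds slight pitch variation around normal.
# It plays quietly through the weapon foley sound class with weapon reload attenuation, with concurrency override and debug both enabled.
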SoundCue /Game/Sounds/Fort_Weapons/GrapplingHook/Grappling_Reload_Insert_Papaya_Cue
root=Modulator_0
nodes [
 {id=Modulator_0, children [Random_0], properties {pitchMax=1.1, pitchMin=0.9}}
 {id=Random_0, children [WavePlayer_0,WavePlayer_1], properties {weights [1,1]}}
 {id=WavePlayer_0, properties {soundWaveAssetPtr=/Game/Sounds/Fort_Weapons/Crossbow/sfx_crossbow_reload_01}}
 {id=WavePlayer_1, properties {soundWaveAssetPtr=/Game/Sounds/Fort_Weapons/Crossbow/sfx_crossbow_reload_02}}
]
properties {
 attenuationSettings=&Fort_Weapon_Reload_Att
 bDebug=true
 bOverrideConcurrency=true
 maxDistance=2250
 soundClassObject=&Fort_WeaponFoley
 volumeMultiplier=0.351
}
