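# Sound cue for the start of the grappling hook reload, named as the Papaya variant, reusing a crossbow reload sample.
# A single wave player feeds the root directly, with no randomization or modulation.
# It plays at reduced volume in the weapon foley sound class, uses the weapon reload attenuation asset, and has concurrency override and debug flags enabled.
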SoundCue /Game/Sounds/Fort_Weapons/GrapplingHook/Grappling_Reload_Start_Papaya_Cue
root=WavePlayer_29
nodes [{id=WavePlayer_29, properties {soundWaveAssetPtr=/Game/Sounds/Fort_Weapons/Crossbow/sfx_crossbow_reload_03}}]
properties {
 attenuationSettings=&Fort_Weapon_Reload_Att
 bDebug=true
 bOverrideConcurrency=true
 maxDistance=2250
 soundClassObject=&Fort_WeaponFoley
 volumeMultiplier=0.301
}
