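# Sound cue for the grappling hook rope being destroyed (Papaya variant), playing a snap layered with a quieter body-swing whoosh.
# A Mixer combines a random pick of four snap waves with a slightly delayed random pick of five body-swing waves, pitched down and softened by a Modulator.
# A root Modulator wraps the mix, which uses grappling gadget attenuation, the gadget items sound class and a two-per-owner, stop-oldest concurrency rule.
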SoundCue /Game/Sounds/Fort_Weapons/GrapplingHook/Grappling_RopeDestroy_Papaya_Cue
root=Modulator_1
nodes [
 {id=Delay_0, children [Modulator_0], properties {delayMax=0.2, delayMin=0.2}}
 {id=Mixer_0, children [Random_0,Delay_0], properties {inputVolume [1.2,0.4]}}
 {id=Modulator_0, children [Random_1], properties {pitchMax=0.9, pitchMin=0.7, volumeMax=0.4, volumeMin=0.3}}
 {id=Modulator_1, children [Mixer_0]}
 {id=Random_0, children [WavePlayer_0,WavePlayer_1,WavePlayer_4,WavePlayer_5], properties {weights [1,1,1,1]}}
 {id=Random_1, children [WavePlayer_6, WavePlayer_7, WavePlayer_8, WavePlayer_9, WavePlayer_10], properties {weights [1,1,1,1,1]}}
 {id=WavePlayer_0, properties {soundWaveAssetPtr=&/Grappling_Snap_01}}
 {id=WavePlayer_1, properties {soundWaveAssetPtr=&/Grappling_Snap_02}}
 {id=WavePlayer_10, properties {soundWaveAssetPtr=&/Grappling_BodySwing_05}}
 {id=WavePlayer_4, properties {soundWaveAssetPtr=&/Grappling_Snap_03}}
 {id=WavePlayer_5, properties {soundWaveAssetPtr=&/Grappling_Snap_04}}
 {id=WavePlayer_6, properties {soundWaveAssetPtr=&/Grappling_BodySwing_01}}
 {id=WavePlayer_7, properties {soundWaveAssetPtr=&/Grappling_BodySwing_02}}
 {id=WavePlayer_8, properties {soundWaveAssetPtr=&/Grappling_BodySwing_03}}
 {id=WavePlayer_9, properties {soundWaveAssetPtr=&/Grappling_BodySwing_04}}
]
properties {
 attenuationSettings=/Game/Sounds/Attenuation/Fort_GadgetGrappling_Papaya_Att
 maxDistance=2300
 pitchMultiplier=1.2
 soundClassObject=&Fort_Gadgets_Items
 volumeMultiplier=0.2
 concurrencySet [/Game/Sounds/Concurrency_Rules/TwoMaximum_Owner_StopOldest]
}
