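# Sound cue for the grappling hook yank in the Fortnite weapons audio set.
# A Random node picks one of three equally weighted yank waves, then a Modulator slightly varies its pitch on each play.
# The cue plays at lowered volume and pitch, routed to the gadget items sound class with grappling-specific attenuation.
# A per-owner concurrency rule caps it at two instances, stopping the oldest.
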SoundCue /Game/Sounds/Fort_Weapons/GrapplingHook/Grappling_Yank_Cue
root=Modulator_1
nodes [
 {id=Modulator_1, children [Random_0], properties {pitchMax=1.1, pitchMin=0.9}}
 {id=Random_0, children [WavePlayer_4,WavePlayer_7,WavePlayer_8], properties {weights [1,1,1]}}
 {id=WavePlayer_4, properties {soundWaveAssetPtr=&/Grappling_Yank_01}}
 {id=WavePlayer_7, properties {soundWaveAssetPtr=&/Grappling_Yank_03}}
 {id=WavePlayer_8, properties {soundWaveAssetPtr=&/Grappling_Yank_02}}
]
properties {
 attenuationSettings=/Game/Sounds/Attenuation/Fort_GadgetGrappling_Att
 maxDistance=8500
 pitchMultiplier=0.9
 soundClassObject=&Fort_Gadgets_Items
 volumeMultiplier=0.45
 concurrencySet [/Game/Sounds/Concurrency_Rules/TwoMaximum_Owner_StopOldest]
}
